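# Control Rig for the winter Llama hero's medium male body that procedurally counter-rotates the head bone.
# It reads head rotation relative to neck_01 and in local space as Euler angles, then clamp-remaps selected pitch or yaw-like axes into opposing offsets.
# Two chained additive-local bone modifications apply these on Y and Z, each responding to one direction of rotation.
# The jaw and horse-hair dynamic bones are in the hierarchy but not driven here.
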ControlRigBlueprint /Game/Characters/Player/Male/Medium/Bodies/M_MED_Llama_Hero_Winter/Meshes/Parts/M_MED_Llama_Hero_Winter_CtrlRig/M_MED_Llama_Hero_Winter_CtrlRig
graphs [
 RigVMFunctionLibrary {links [], nodes []}
 RigVMModel {nodes [
   RigUnit_BeginExecution = &RigUnit_BeginExecution()
   n1 = &RigVMFunction_MathQuaternionToEuler(RotationOrder=ZYX, Value=&RigUnit_GetRelativeBoneTransform(Bone=head, Space=neck_01).Transform.Rotation)
   n2 = &RigVMFunction_MathQuaternionToEuler(RotationOrder=ZYX, Value=&RigUnit_GetBoneTransform(Bone=head, Space=LocalSpace).Transform.Rotation)
   /Script/ControlRig.RigUnit_ModifyBoneTransforms(
    BoneToModify=((Bone="head",Transform=(Rotation=(X=0,Y=0,Z=0,W=1),Translation=(X=0,Y=0,Z=0),Scale3D=(X=1,Y=1,Z=1))))
    Mode=AdditiveLocal
    Weight=1
    WeightMaximum=1
    WeightMinimum=0
    BoneToModify.0.Transform.Rotation=&RigVMFunction_MathQuaternionFromEuler(
     Euler=(X=0,Y=0,Z=0)
     Euler.Z=&RigVMFunction_MathFloatRemap(SourceMaximum=-50, SourceMinimum=-25, TargetMaximum=30, TargetMinimum=0, bClamp=true, Value=@n1.Result.Z).Result
     Euler.Y=&RigVMFunction_MathFloatRemap(SourceMaximum=-10, SourceMinimum=0, TargetMaximum=15, TargetMinimum=0, bClamp=true, Value=@n2.Result.Y).Result
    ).Result
    ExecutePin=@RigUnit_BeginExecution.ExecutePin
   ) {is=struct RigVMUnitNode}
   > &RigUnit_ModifyBoneTransforms(
    BoneToModify=((Bone="head",Transform=(Rotation=(X=0,Y=0,Z=0,W=1),Translation=(X=0,Y=0,Z=0),Scale3D=(X=1,Y=1,Z=1))))
    Mode=AdditiveLocal
    Weight=1
    WeightMaximum=1
    WeightMinimum=0
    BoneToModify.0.Transform.Rotation=&RigVMFunction_MathQuaternionFromEuler(
     Euler=(X=0,Y=0,Z=0)
     Euler.Z=&RigVMFunction_MathFloatRemap(SourceMaximum=10, SourceMinimum=0, TargetMaximum=-15, TargetMinimum=0, bClamp=true, Value=@n2.Result.Z).Result
    ).Result
   )
  ]}
]
hierarchy [
 root {loc=<0 0 0>, rot=<0 0 0 1>}
 pelvis {parent=root, loc=<0 -0.7071 93.1191>, rot=<0.0127 0.707 -0.0127 -0.707>}
 spine_01 {parent=pelvis, loc=<6.7494 -0.6388 0>, rot=<0 0 0.034 -0.9994>}
 spine_02 {parent=spine_01, loc=<7.2518 -0.035 0>, rot=<0 0 0.0238 -0.9997>}
 spine_03 {parent=spine_02, loc=<6.0104 0 0>, rot=<0 0 0.0149 -0.9999>}
 spine_04 {parent=spine_03, loc=<8.1024 0 0>, rot=<0 0 -0.0132 -0.9999>}
 spine_05 {parent=spine_04, loc=<10.0146 0 0>, rot=<0 0 -0.0015 -1>}
 clavicle_l {parent=spine_05, loc=<4.3575 -0.1466 -2.3223>, rot=<0.0465 -0.7289 -0.009 -0.6829>}
 upperarm_l {parent=clavicle_l, loc=<15.5111 0 0>, rot=<0.0305 -0.536 -0.0057 -0.8437>}
 upperarm_twist_01_l {parent=upperarm_l, loc=<0.0041 0 0>, rot=<0 0 0 1>}
 clavicle_r {parent=spine_05, loc=<4.3575 -0.1466 2.3223>, rot=<0.7289 0.0465 -0.6829 0.009>}
 upperarm_r {parent=clavicle_r, loc=<-15.5111 0 0>, rot=<0.0305 -0.536 -0.0057 -0.8437>}
 upperarm_twist_01_r {parent=upperarm_r, loc=<-0.0041 0 0>, rot=<0 0 0 1>}
 neck_01 {parent=spine_05, loc=<11.3138 0.1529 0>, rot=<0 0 -0.2003 -0.9797>}
 neck_02 {parent=neck_01, loc=<4.5627 0 0>, rot=<0 0 0.0504 -0.9987>}
 head {parent=neck_02, loc=<5.3914 0 0>, rot=<0 0 0.115 -0.9934>}
 dyn_C_jaw_ALT {parent=head, loc=<-0.3094 13.6845 0.0151>, rot=<0.0082 -0.7071 -0.0082 -0.7071>}
 dyn_horse_hair_01 {parent=head, loc=<17.042 4.0165 0.0151>, rot=<0.6967 0.1207 -0.1207 -0.6967>}
 dyn_horse_hair_02 {parent=head, loc=<16.5304 1.8266 0.0151>, rot=<0.7061 -0.0381 0.0381 -0.7061>}
 dyn_horse_hair_04 {parent=dyn_horse_hair_02, loc=<4.173 0 -1.9547>, rot=<0 -0.1342 0 -0.991>}
 dyn_horse_hair_03 {parent=dyn_horse_hair_02, loc=<4.5149 0 -0.8798>, rot=<0 -0.0683 0 -0.9977>}
 dyn_horse_hair_05 {parent=head, loc=<12.7063 -2.0073 0.0151>, rot=<0.6286 -0.3238 0.3238 -0.6286>}
 dyn_horse_hair_06 {parent=head, loc=<8.2597 -4.6198 0.0151>, rot=<0.4716 -0.5269 0.5269 -0.4716>}
 dyn_horse_hair_07 {parent=head, loc=<3.236 -6.6456 0.0151>, rot=<0.3883 -0.5909 0.5909 -0.3883>}
]
